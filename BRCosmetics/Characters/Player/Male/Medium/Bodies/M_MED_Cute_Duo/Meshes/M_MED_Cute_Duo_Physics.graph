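# Physics asset for the M_MED_Cute_Duo medium male player cosmetic body, covering its dynamic lower-shirt bones.
# Pelvis, thigh and left thigh-twist capsules form the body, with box panels for left side, left front and both back shirt pieces.
# Each panel hinges with limited angles on a tiny MSTR anchor body, itself constrained to its thigh with free rotation.
# The left side and front panels are also linked by a free-rotation constraint.
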PhysicsAsset /BRCosmetics/Characters/Player/Male/Medium/Bodies/M_MED_Cute_Duo/Meshes/M_MED_Cute_Duo_Physics
bodies {
 dyn_lowershirt_01_l [{shapeType=Box, extentX=20.5, extentY=7, extentZ=17.5, center=<10.4605 0 0.0007>}]
 dyn_lowershirt_01_l_MSTR [{shapeType=Capsule, length=0.01, radius=0.01}]
 dyn_lowershirt_bk_01_l [{shapeType=Box, extentX=20.5, extentY=6, extentZ=17.5, center=<10.4605 0 0.0007>}]
 dyn_lowershirt_bk_01_l_MSTR [{shapeType=Capsule, length=0.01, radius=0.01}]
 dyn_lowershirt_bk_01_r [{shapeType=Box, extentX=20.5, extentY=6, extentZ=17.5, center=<10.4605 0 0.0007>}]
 dyn_lowershirt_bk_01_r_MSTR [{shapeType=Capsule, length=0.01, radius=0.01}]
 dyn_lowershirt_fr_01_l [{shapeType=Box, extentX=20.5, extentY=6, extentZ=17.5, center=<10.4605 0 0.0007>}]
 dyn_lowershirt_fr_01_l_MSTR [{shapeType=Capsule, length=0.01, radius=0.01}]
 pelvis [{shapeType=Capsule, length=14.8027, radius=9.1635}]
 thigh_l [{shapeType=Capsule, length=38.102, radius=6.5021, center=<-21.9405 0 0>, rotation=<rot -90 -0.0149 0.0149>}]
 thigh_r [{shapeType=Capsule, length=38.102, radius=6.5021, center=<21.9405 0 0>, rotation=<rot -90 0.0197 179.9803>}]
 thigh_twist_01_l [{shapeType=Capsule, length=14.602, radius=8.0021, center=<9.1588 0.0004 0>, rotation=<rot -90 0.0197 -0.0197>}]
}
constraints [
 {bone1Name=thigh_l, bone2Name=pelvis, swing1LimitDegrees=45, swing2LimitDegrees=45, twistLimitDegrees=45}
 {bone1Name=thigh_r, bone2Name=pelvis, swing1LimitDegrees=45, swing2LimitDegrees=45, twistLimitDegrees=45}
 {
  bone1Name=dyn_lowershirt_bk_01_l_MSTR
  bone2Name=thigh_l
  swing1LimitDegrees=45
  swing1Motion=Free
  swing2LimitDegrees=45
  swing2Motion=Free
  twistLimitDegrees=45
  twistMotion=Free
 }
 {
  bone1Name=dyn_lowershirt_fr_01_l_MSTR
  bone2Name=thigh_l
  swing1LimitDegrees=45
  swing1Motion=Free
  swing2LimitDegrees=45
  swing2Motion=Free
  twistLimitDegrees=45
  twistMotion=Free
 }
 {
  bone1Name=dyn_lowershirt_01_l_MSTR
  bone2Name=thigh_l
  swing1LimitDegrees=45
  swing1Motion=Free
  swing2LimitDegrees=45
  swing2Motion=Free
  twistLimitDegrees=45
  twistMotion=Free
 }
 {
  bone1Name=dyn_lowershirt_bk_01_r_MSTR
  bone2Name=thigh_r
  swing1LimitDegrees=45
  swing1Motion=Free
  swing2LimitDegrees=45
  swing2Motion=Free
  twistLimitDegrees=45
  twistMotion=Free
 }
 {bone1Name=dyn_lowershirt_bk_01_l, bone2Name=dyn_lowershirt_bk_01_l_MSTR, swing1LimitDegrees=45, swing2LimitDegrees=5, twistLimitDegrees=5}
 {bone1Name=dyn_lowershirt_fr_01_l, bone2Name=dyn_lowershirt_fr_01_l_MSTR, swing1LimitDegrees=75, swing2LimitDegrees=5, twistLimitDegrees=5}
 {bone1Name=dyn_lowershirt_01_l, bone2Name=dyn_lowershirt_01_l_MSTR, swing1LimitDegrees=25, swing2LimitDegrees=25, twistLimitDegrees=5}
 {bone1Name=dyn_lowershirt_bk_01_r, bone2Name=dyn_lowershirt_bk_01_r_MSTR, swing1LimitDegrees=75, swing2LimitDegrees=75, twistLimitDegrees=5}
 {
  bone1Name=dyn_lowershirt_01_l
  bone2Name=dyn_lowershirt_fr_01_l
  swing1LimitDegrees=45
  swing1Motion=Free
  swing2LimitDegrees=45
  swing2Motion=Free
  twistLimitDegrees=45
  twistMotion=Free
 }
 {
  bone1Name=thigh_twist_01_l
  bone2Name=thigh_l
  swing1LimitDegrees=45
  swing1Motion=Free
  swing2LimitDegrees=45
  swing2Motion=Free
  twistLimitDegrees=45
  twistMotion=Free
 }
]
