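# Post-pose AnimBlueprint for the M_MED_Cute_Duo cosmetic body, adding secondary motion and corrective bones to a pose copied from the parent mesh.
# Constraints, RigidBody and AnimDynamics drive shirt, sleeves, laces and hoodie.
# EventGraph turns crouch and upperarm twist into alphas for later corrective transforms and constraints.
# RshoulderAlphaBicep takes the left-arm twist, right-foot laces reference foot_l, and RbicepAlpha is never read in AnimGraph.
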AnimBlueprint /BRCosmetics/Characters/Player/Male/Medium/Bodies/M_MED_Cute_Duo/Meshes/M_MED_Cute_Duo_AnimBP
target_skeleton=&/M_MED_Cute_Duo_Skeleton
graphs [
 AnimGraph {nodes [
   n1 = "Constraint - dyn_sleeve_01_l_MSTR (P)"(ComponentPose=Local To Component(LocalPose=Copy Pose From Mesh() {bUseAttachedParent=true}), Alpha=1, ConstraintWeights_0=1) {
    alphaBoolBlend {blendOption=Linear}
    boneToModify {boneName=dyn_sleeve_01_l_MSTR}
    constraintSetup [{targetBone {boneName=upperarm_twist_01_l}}]
    constraintWeights [1]
   }
   n2 = "Constraint - dyn_sleeve_01_r_MSTR (P)"(ComponentPose=@n1, Alpha=1, ConstraintWeights_0=1) {
    alphaBoolBlend {blendOption=Linear}
    boneToModify {boneName=dyn_sleeve_01_r_MSTR}
    constraintSetup [{targetBone {boneName=upperarm_twist_01_r}}]
    constraintWeights [1]
   }
   n3 = "Constraint - dyn_lowershirt_fr_01_l_MSTR (P)"(ComponentPose=@n2, Alpha=1, ConstraintWeights_0=1) {
    alphaBoolBlend {blendOption=Linear}
    boneToModify {boneName=dyn_lowershirt_fr_01_l_MSTR}
    constraintSetup [{targetBone {boneName=thigh_twist_01_l}}]
    constraintWeights [1]
   }
   n4 = "Constraint - dyn_lowershirt_bk_01_l_MSTR (PP)"!(ComponentPose=@n3, Alpha=1, ConstraintWeights_0=1, ConstraintWeights_1:float=0) -> (Pose) {
    alphaBoolBlend {blendOption=Linear}
    boneToModify {boneName=dyn_lowershirt_bk_01_l_MSTR}
    constraintSetup [{targetBone {boneName=pelvis}}, {targetBone {boneName=thigh_l}}]
    constraintWeights [1,1]
   }
   n5 = "Constraint - dyn_lowershirt_bk_01_l_MSTR (RR)"!(ComponentPose=@n4, Alpha=1, ConstraintWeights_0=0.5, ConstraintWeights_1:float=0.5) -> (Pose) {
    alphaBoolBlend {blendOption=Linear}
    boneToModify {boneName=dyn_lowershirt_bk_01_l_MSTR}
    constraintSetup [{transformType=Rotation, targetBone {boneName=thigh_twist_01_l}}, {transformType=Rotation, targetBone {boneName=pelvis}}]
    constraintWeights [1,1]
   }
   n6 = "Constraint - dyn_lowershirt_01_l_MSTR (P)"(ComponentPose=@n5, Alpha=1, ConstraintWeights_0=1) {
    alphaBoolBlend {blendOption=Linear}
    boneToModify {boneName=dyn_lowershirt_01_l_MSTR}
    constraintSetup [{targetBone {boneName=thigh_twist_01_l}}]
    constraintWeights [1]
   }
   n7 = "Transform (Modify) Bone - Bone: dyn_lowershirt_fr_01_l_MSTR"(ComponentPose=@n6, Alpha=1, Translation="0.000,-1.168,0.000", Rotation="0,0,0", Scale="1,1,1") {
    translationMode=BMM_Additive
    translationSpace=BCS_BoneSpace
    alphaBoolBlend {blendOption=Linear}
    boneToModify {boneName=dyn_lowershirt_fr_01_l_MSTR}
    translation {y=-1.168}
   }
   n8 = "Transform (Modify) Bone - Bone: dyn_lowershirt_01_l_MSTR"(ComponentPose=@n7, Alpha=0, Translation="0.000,-5,-0.000", Rotation="0,0,0", Scale="1,1,1") {
    translationMode=BMM_Additive
    translationSpace=BCS_BoneSpace
    alphaBoolBlend {blendOption=Linear}
    boneToModify {boneName=dyn_lowershirt_01_l_MSTR}
    translation {y=-5}
   }
   n9 = RigidBody(ComponentPose=@n8) {
    bClampLinearTranslationLimitToRefPose=true
    bForceDisableCollisionBetweenConstraintBodies=true
    bOverrideWorldGravity=true
    overridePhysicsAsset=&/M_MED_Cute_Duo_Physics
    simulationSpace=BaseBoneSpace
    alphaBoolBlend {blendOption=Linear}
    baseBoneRef {boneName=root}
    componentAppliedLinearAccClamp=<3000 3000 3000>
    componentLinearAccScale=<0.3 0.3 0.6>
    componentLinearVelScale=<0.6 0.6 0.6>
    overrideWorldGravity {z=-980}
   }
   n10 = Anim Dynamics - Bone: dyn_sleeve_01_r(ComponentPose=@n9, Alpha=1, ExternalForce="0,0,0") {
    angularDampingOverride=1
    angularSpringConstant=1500
    bAngularSpring=true
    bGravityOverrideInSimSpace=true
    bOverrideAngularDamping=true
    gravityScale=0.35
    simulationSpace=RootRelative
    alphaBoolBlend {blendOption=Linear}
    boundBone {boneName=dyn_sleeve_01_r}
    physicsBodyDefinitions [{
      sphereCollisionRadius=0
      boundBone {boneName=dyn_sleeve_01_r}
      boxExtents=<10 20 20>
      constraintSetup {angularLimitsMax=<1 1 2>, angularLimitsMin=<-1 -1 -2>, angularTarget=<0 0 0>}
      localJointOffset=<10 0 0>
     }]
   }
   n11 = Anim Dynamics - Bone: dyn_sleeve_01_l(ComponentPose=@n10, Alpha=1, ExternalForce="0,0,0") {
    angularDampingOverride=1
    angularSpringConstant=1500
    bAngularSpring=true
    bGravityOverrideInSimSpace=true
    bOverrideAngularDamping=true
    gravityScale=0.35
    simulationSpace=RootRelative
    alphaBoolBlend {blendOption=Linear}
    boundBone {boneName=dyn_sleeve_01_l}
    physicsBodyDefinitions [{
      sphereCollisionRadius=0
      boundBone {boneName=dyn_sleeve_01_l}
      boxExtents=<10 20 20>
      constraintSetup {angularLimitsMax=<0 0 2>, angularLimitsMin=<0 0 -2>, angularTarget=<0 0 0>}
      localJointOffset=<10 0 0>
     }]
   }
   n12 = Anim Dynamics - Bone: dyn_shoelaceB_01_l(ComponentPose=@n11, Alpha=1, ExternalForce="0,0,0") {
    angularDampingOverride=1
    angularSpringConstant=1500
    bAngularSpring=true
    bGravityOverrideInSimSpace=true
    bOverrideAngularDamping=true
    gravityScale=0.35
    simulationSpace=RootRelative
    alphaBoolBlend {blendOption=Linear}
    boundBone {boneName=dyn_shoelaceB_01_l}
    componentLinearAccScale=<0.3 0.3 0.6>
    componentLinearVelScale=<0.3 0.3 0.6>
    physicsBodyDefinitions [{
      sphereCollisionRadius=0
      boundBone {boneName=dyn_shoelaceB_01_l}
      boxExtents=<10 10 10>
      constraintSetup {angularLimitsMax=<0 5 0>, angularLimitsMin=<0 -5 -60>, angularTarget=<1 0 0>}
      localJointOffset=<5 0 0>
     }]
    relativeSpaceBone {boneName=foot_l}
   }
   n13 = Anim Dynamics - Bone: dyn_shoelaceA_01_l(ComponentPose=@n12, Alpha=1, ExternalForce="0,0,0") {
    angularDampingOverride=1
    angularSpringConstant=1500
    bAngularSpring=true
    bGravityOverrideInSimSpace=true
    bOverrideAngularDamping=true
    gravityScale=0.35
    simulationSpace=RootRelative
    alphaBoolBlend {blendOption=Linear}
    boundBone {boneName=dyn_shoelaceA_01_l}
    componentLinearAccScale=<0.3 0.3 0.6>
    componentLinearVelScale=<0.3 0.3 0.6>
    physicsBodyDefinitions [{
      sphereCollisionRadius=0
      boundBone {boneName=dyn_shoelaceA_01_l}
      boxExtents=<10 10 10>
      constraintSetup {angularLimitsMax=<0 5 60>, angularLimitsMin=<0 -5 0>, angularTarget=<1 0 0>}
      localJointOffset=<5 0 0>
     }]
    relativeSpaceBone {boneName=foot_l}
   }
   n14 = Anim Dynamics - Bone: dyn_shoelaceB_01_r(ComponentPose=@n13, Alpha=1, ExternalForce="0,0,0") {
    angularDampingOverride=1
    angularSpringConstant=1500
    bAngularSpring=true
    bGravityOverrideInSimSpace=true
    bOverrideAngularDamping=true
    gravityScale=0.35
    simulationSpace=RootRelative
    alphaBoolBlend {blendOption=Linear}
    boundBone {boneName=dyn_shoelaceB_01_r}
    componentLinearAccScale=<0.3 0.3 0.6>
    componentLinearVelScale=<0.3 0.3 0.6>
    physicsBodyDefinitions [{
      sphereCollisionRadius=0
      boundBone {boneName=dyn_shoelaceB_01_r}
      boxExtents=<10 10 10>
      constraintSetup {angularLimitsMax=<0 5 60>, angularLimitsMin=<0 -5 0>, angularTarget=<1 0 0>}
      localJointOffset=<5 0 0>
     }]
    relativeSpaceBone {boneName=foot_l}
   }
   n15 = Anim Dynamics - Bone: dyn_shoelaceA_01_r(ComponentPose=@n14, Alpha=1, ExternalForce="0,0,0") {
    angularDampingOverride=1
    angularSpringConstant=1500
    bAngularSpring=true
    bGravityOverrideInSimSpace=true
    bOverrideAngularDamping=true
    gravityScale=0.35
    simulationSpace=RootRelative
    alphaBoolBlend {blendOption=Linear}
    boundBone {boneName=dyn_shoelaceA_01_r}
    componentLinearAccScale=<0.3 0.3 0.6>
    componentLinearVelScale=<0.3 0.3 0.6>
    physicsBodyDefinitions [{
      sphereCollisionRadius=0
      boundBone {boneName=dyn_shoelaceA_01_r}
      boxExtents=<10 10 10>
      constraintSetup {angularLimitsMax=<0 5 0>, angularLimitsMin=<0 -5 -60>, angularTarget=<1 0 0>}
      localJointOffset=<5 0 0>
     }]
    relativeSpaceBone {boneName=foot_l}
   }
   n16 = Anim Dynamics - Bone: dyn_upperhoodie_01_l!(ComponentPose=@n15, Alpha=1, GravityOverride:Vector="0,0,0", ExternalForce="0,0,0") -> (Pose) {
    angularDampingOverride=1
    angularSpringConstant=1500
    bAngularSpring=true
    bGravityOverrideInSimSpace=true
    bOverrideAngularDamping=true
    bUseGravityOverride=true
    gravityScale=0.34
    simulationSpace=RootRelative
    alphaBoolBlend {blendOption=Linear}
    boundBone {boneName=dyn_upperhoodie_01_l}
    componentLinearAccScale=<0.3 0.3 0.5>
    componentLinearVelScale=<0.5 0.5 0.7>
    gravityOverride {z=-980}
    physicsBodyDefinitions [{
      sphereCollisionRadius=0
      boundBone {boneName=dyn_upperhoodie_01_l}
      boxExtents=<20 20 20>
      constraintSetup {angularLimitsMax=<0 0 15>, angularLimitsMin=<0 -15 -15>, angularTarget=<0 0 0>}
      localJointOffset=<8 0 0>
     }]
    relativeSpaceBone {boneName=spine_05}
   }
   n17 = Anim Dynamics - Bone: dyn_upperhoodie_fr_01_l!(ComponentPose=@n16, Alpha=1, GravityOverride:Vector="0,0,0", ExternalForce="0,0,0") -> (Pose) {
    angularDampingOverride=1
    angularSpringConstant=1500
    bAngularSpring=true
    bGravityOverrideInSimSpace=true
    bOverrideAngularDamping=true
    bUseGravityOverride=true
    gravityScale=0.34
    simulationSpace=RootRelative
    alphaBoolBlend {blendOption=Linear}
    boundBone {boneName=dyn_upperhoodie_fr_01_l}
    componentLinearAccScale=<0.3 0.3 0.5>
    componentLinearVelScale=<0.5 0.5 0.7>
    gravityOverride {z=-980}
    physicsBodyDefinitions [{
      sphereCollisionRadius=0
      boundBone {boneName=dyn_upperhoodie_fr_01_l}
      boxExtents=<20 20 20>
      constraintSetup {angularLimitsMax=<0 0 15>, angularLimitsMin=<0 -15 -15>, angularTarget=<0 0 0>}
      localJointOffset=<8 0 0>
     }]
    relativeSpaceBone {boneName=spine_05}
   }
   n18 = Anim Dynamics - Bone: dyn_lowerhoodie_bk_00_l!(ComponentPose=@n17, Alpha=1, GravityOverride:Vector="0,0,0", ExternalForce="0,0,0") -> (Pose) {
    angularDampingOverride=1
    angularSpringConstant=1500
    bAngularSpring=true
    bGravityOverrideInSimSpace=true
    bOverrideAngularDamping=true
    bUseGravityOverride=true
    gravityScale=0.34
    simulationSpace=RootRelative
    alphaBoolBlend {blendOption=Linear}
    boundBone {boneName=dyn_lowerhoodie_bk_00_l}
    componentLinearAccScale=<0.3 0.3 0.5>
    componentLinearVelScale=<0.5 0.5 0.7>
    gravityOverride {z=-980}
    physicsBodyDefinitions [{
      sphereCollisionRadius=0
      boundBone {boneName=dyn_lowerhoodie_bk_00_l}
      boxExtents=<20 20 20>
      constraintSetup {angularLimitsMax=<0 0 15>, angularLimitsMin=<0 -15 -15>, angularTarget=<0 0 0>}
      localJointOffset=<8 0 0>
     }]
    relativeSpaceBone {boneName=spine_04}
   }
   n19 = Anim Dynamics - Bone: dyn_lowerhoodie_fr_00_l!(ComponentPose=@n18, Alpha=1, GravityOverride:Vector="0,0,0", ExternalForce="0,0,0") -> (Pose) {
    angularDampingOverride=1
    angularSpringConstant=1500
    bAngularSpring=true
    bGravityOverrideInSimSpace=true
    bOverrideAngularDamping=true
    bUseGravityOverride=true
    gravityScale=0.34
    simulationSpace=RootRelative
    alphaBoolBlend {blendOption=Linear}
    boundBone {boneName=dyn_lowerhoodie_fr_00_l}
    componentLinearAccScale=<0.3 0.3 0.5>
    componentLinearVelScale=<0.5 0.5 0.7>
    gravityOverride {z=-980}
    physicsBodyDefinitions [{
      sphereCollisionRadius=0
      boundBone {boneName=dyn_lowerhoodie_fr_00_l}
      boxExtents=<20 20 20>
      constraintSetup {angularLimitsMax=<0 0 15>, angularLimitsMin=<0 -15 -15>, angularTarget=<0 0 0>}
      localJointOffset=<8 0 0>
     }]
    relativeSpaceBone {boneName=spine_05}
   }
   n20 = Anim Dynamics - Bone: dyn_upperhoodie_fr_01_r!(ComponentPose=@n19, Alpha=1, GravityOverride:Vector="0,0,0", ExternalForce="0,0,0") -> (Pose) {
    angularDampingOverride=1
    angularSpringConstant=1500
    bAngularSpring=true
    bGravityOverrideInSimSpace=true
    bOverrideAngularDamping=true
    bUseGravityOverride=true
    gravityScale=0.34
    simulationSpace=RootRelative
    alphaBoolBlend {blendOption=Linear}
    boundBone {boneName=dyn_upperhoodie_fr_01_r}
    componentLinearAccScale=<0.3 0.3 0.5>
    componentLinearVelScale=<0.5 0.5 0.7>
    gravityOverride {z=-980}
    physicsBodyDefinitions [{
      sphereCollisionRadius=0
      boundBone {boneName=dyn_upperhoodie_fr_01_r}
      boxExtents=<20 20 20>
      constraintSetup {angularLimitsMax=<0 0 15>, angularLimitsMin=<0 -15 -15>, angularTarget=<0 0 0>}
      localJointOffset=<8 0 0>
     }]
    relativeSpaceBone {boneName=spine_05}
   }
   n21 = Anim Dynamics - Bone: dyn_upperhoodie_01_r!(ComponentPose=@n20, Alpha=1, GravityOverride:Vector="0,0,0", ExternalForce="0,0,0") -> (Pose) {
    angularDampingOverride=1
    angularSpringConstant=1500
    bAngularSpring=true
    bGravityOverrideInSimSpace=true
    bOverrideAngularDamping=true
    bUseGravityOverride=true
    gravityScale=0.34
    simulationSpace=RootRelative
    alphaBoolBlend {blendOption=Linear}
    boundBone {boneName=dyn_upperhoodie_01_r}
    componentLinearAccScale=<0.3 0.3 0.5>
    componentLinearVelScale=<0.5 0.5 0.7>
    gravityOverride {z=-980}
    physicsBodyDefinitions [{
      sphereCollisionRadius=0
      boundBone {boneName=dyn_upperhoodie_01_r}
      boxExtents=<20 20 20>
      constraintSetup {angularLimitsMax=<0 0 15>, angularLimitsMin=<0 -15 -15>, angularTarget=<0 0 0>}
      localJointOffset=<8 0 0>
     }]
    relativeSpaceBone {boneName=spine_05}
   }
   n22 = Anim Dynamics - Bone: dyn_lowerhoodie_bk_00_r!(ComponentPose=@n21, Alpha=1, GravityOverride:Vector="0,0,0", ExternalForce="0,0,0") -> (Pose) {
    angularDampingOverride=1
    angularSpringConstant=1500
    bAngularSpring=true
    bGravityOverrideInSimSpace=true
    bOverrideAngularDamping=true
    bUseGravityOverride=true
    gravityScale=0.34
    simulationSpace=RootRelative
    alphaBoolBlend {blendOption=Linear}
    boundBone {boneName=dyn_lowerhoodie_bk_00_r}
    componentLinearAccScale=<0.3 0.3 0.5>
    componentLinearVelScale=<0.5 0.5 0.7>
    gravityOverride {z=-980}
    physicsBodyDefinitions [{
      sphereCollisionRadius=0
      boundBone {boneName=dyn_lowerhoodie_bk_00_r}
      boxExtents=<20 20 20>
      constraintSetup {angularLimitsMax=<0 0 15>, angularLimitsMin=<0 -15 -15>, angularTarget=<0 0 0>}
      localJointOffset=<8 0 0>
     }]
    relativeSpaceBone {boneName=spine_04}
   }
   n23 = Anim Dynamics - Bone: dyn_lowerhoodie_fr_00_r!(ComponentPose=@n22, Alpha=1, GravityOverride:Vector="0,0,0", ExternalForce="0,0,0") -> (Pose) {
    angularDampingOverride=1
    angularSpringConstant=1500
    bAngularSpring=true
    bGravityOverrideInSimSpace=true
    bOverrideAngularDamping=true
    bUseGravityOverride=true
    gravityScale=0.34
    simulationSpace=RootRelative
    alphaBoolBlend {blendOption=Linear}
    boundBone {boneName=dyn_lowerhoodie_fr_00_r}
    componentLinearAccScale=<0.3 0.3 0.5>
    componentLinearVelScale=<0.5 0.5 0.7>
    gravityOverride {z=-980}
    physicsBodyDefinitions [{
      sphereCollisionRadius=0
      boundBone {boneName=dyn_lowerhoodie_fr_00_r}
      boxExtents=<20 20 20>
      constraintSetup {angularLimitsMax=<0 0 15>, angularLimitsMin=<0 -15 -15>, angularTarget=<0 0 0>}
      localJointOffset=<8 0 0>
     }]
    relativeSpaceBone {boneName=spine_05}
   }
   n24 = "Constraint - dyn_sleeve_bt_01_l (RR)"!(ComponentPose=@n23, Alpha=Get LshoulderAlpha() -> (LshoulderAlpha:double), ConstraintWeights_0=0.5, ConstraintWeights_1:float=0.5) -> (Pose) {
    alphaBoolBlend {blendOption=Linear}
    boneToModify {boneName=dyn_sleeve_bt_01_l}
    constraintSetup [{transformType=Rotation, targetBone {boneName=dyn_sleeve_01_l}}, {transformType=Rotation, targetBone {boneName=spine_04}}]
    constraintWeights [1,1]
   }
   n25 = "Constraint - dyn_sleeve_bt_01_r (RR)"!(ComponentPose=@n24, Alpha=Get RshoulderAlpha() -> (RshoulderAlpha:double), ConstraintWeights_0=0.5, ConstraintWeights_1:float=0.5) -> (Pose) {
    alphaBoolBlend {blendOption=Linear}
    boneToModify {boneName=dyn_sleeve_bt_01_r}
    constraintSetup [{transformType=Rotation, targetBone {boneName=dyn_sleeve_01_r}}, {transformType=Rotation, targetBone {boneName=spine_04}}]
    constraintWeights [1,1]
   }
   n26 = "Transform (Modify) Bone - Bone: dyn_bicep_01_r"(ComponentPose=@n25, Alpha=Get RshoulderAlpha(), Translation="0.000,10,0.000", Rotation="0,0,0", Scale="1,1,1") {
    translationMode=BMM_Additive
    translationSpace=BCS_BoneSpace
    alphaBoolBlend {blendOption=Linear}
    boneToModify {boneName=dyn_bicep_01_r}
    translation {y=10}
   }
   n27 = "Constraint - dyn_bicep_01_r (P)"(ComponentPose=@n26, Alpha=Get RshoulderAlphaBicep() -> (RshoulderAlphaBicep:double), ConstraintWeights_0=1) {
    alphaBoolBlend {blendOption=Linear}
    boneToModify {boneName=dyn_bicep_01_r}
    constraintSetup [{targetBone {boneName=dyn_sleeve_bt_01_r}}]
    constraintWeights [1]
   }
   n28 = Get crouchingAlpha() -> (crouchingAlpha:double)
   n29 = "Transform (Modify) Bone - Bone: dyn_thighhelper_01_l"(ComponentPose=@n27, Alpha=@n28, Translation="0.000,5.522,0.000", Rotation="0,0,0", Scale="1,1,1") {
    translationMode=BMM_Additive
    translationSpace=BCS_BoneSpace
    alphaBoolBlend {blendOption=Linear}
    boneToModify {boneName=dyn_thighhelper_01_l}
    translation {y=5.522}
   }
   n30 = "Transform (Modify) Bone - Bone: dyn_lowershirt_fr_01_l_MSTR"(ComponentPose=@n29, Alpha=@n28, Translation="5.807,-3.219,0.000", Rotation="0.000,15.000,0.000", Scale="1,1,1") {
    rotationMode=BMM_Additive
    rotationSpace=BCS_BoneSpace
    translationMode=BMM_Additive
    translationSpace=BCS_BoneSpace
    alphaBoolBlend {blendOption=Linear}
    boneToModify {boneName=dyn_lowershirt_fr_01_l_MSTR}
    rotation {yaw=15}
    translation=<5.807 -3.219>
   }
   n31 = "Transform (Modify) Bone - Bone: dyn_lowershirt_fr_01_l_MSTR"(ComponentPose=@n30, Alpha=0, Translation="5.807,-3.219,0.000", Rotation="0.000,15.000,0.000", Scale="1,1,1") {
    rotationMode=BMM_Additive
    rotationSpace=BCS_BoneSpace
    translationMode=BMM_Additive
    translationSpace=BCS_BoneSpace
    alphaBoolBlend {blendOption=Linear}
    boneToModify {boneName=dyn_lowershirt_fr_01_l_MSTR}
    rotation {yaw=15}
    translation=<5.807 -3.219>
   }
   n32 = "Constraint - dyn_lowershirt_bk_01_r (PP)"!(ComponentPose=@n31, Alpha=1, ConstraintWeights_0=0.5, ConstraintWeights_1:float=0.5) -> (Pose) {
    alphaBoolBlend {blendOption=Linear}
    boneToModify {boneName=dyn_lowershirt_bk_01_r}
    constraintSetup [{targetBone {boneName=pelvis}}, {targetBone {boneName=thigh_r}}]
    constraintWeights [1,1]
   }
   Output Pose(Result=Component To Local(ComponentPose=@n32)) {layerGroup=None}
  ]}
 EventGraph {nodes [
   Event BlueprintUpdateAnimation()
   > n1 = Sequence!(execute) -> (then_0, then_1, then_2:Exec, then_3:Exec)
   > Set crouchingAlpha(crouchingAlpha:double="To Float (Boolean)"(InBool=Get bIsCrouching())) -> (Output_Get:double)
   n2 = GetOwnerSkeletalMeshComponent()
   n3 = GetAlphaFromReferenceTransformDelta(
    SkeletalMeshComponent=@n2
    BoneName=upperarm_l
    BaseName=clavicle_l
    DeltaComponentType=QuaternionTwist
    InputRange=(X=-30,Y=20)
    TwistAxis="1,0,0"
   )
   Set LshoulderAlpha(execute=@n1.then_1, LshoulderAlpha:double=float * float(A=@n3, B=0.25)) -> (Output_Get:double)
   > Set RshoulderAlphaBicep(RshoulderAlphaBicep:double=@n3) -> (Output_Get:double)
   > Set RshoulderAlpha(RshoulderAlpha:double=float * float(
     A=GetAlphaFromReferenceTransformDelta(
      SkeletalMeshComponent=@n2
      BoneName=upperarm_r
      BaseName=clavicle_r
      DeltaComponentType=QuaternionTwist
      InputRange=(X=-30,Y=20)
      TwistAxis="1,0,0"
     )
     B=0.25
    )) -> (Output_Get:double)
   > Set RbicepAlpha(RbicepAlpha:double=SelectFloat(
     A=1
     B=GetAlphaFromReferenceTransformDelta(
      SkeletalMeshComponent=@n2
      BoneName=upperarm_r
      BaseName=clavicle_r
      DeltaComponentType=QuaternionTwist
      InputRange=(X=-30,Y=20)
      TwistAxis="1,0,0"
     )
     bPickA=Get bIsTargeting()
    )) -> (Output_Get:double)
  ]}
]
variables [
 crouchingAlpha {type=number}
 LshoulderAlpha {type=number}
 RbicepAlpha {type=number}
 RshoulderAlpha {type=number}
 RshoulderAlphaBicep {type=number}
]
