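# Physics asset for the face accessory of the Cute_Duo medium male player cosmetic, covering the head and left and right "poof" dynamic bones.
# Each side chains from a head capsule through a tiny MSTR capsule to a larger poof capsule.
# The MSTR-to-head joints swing and twist freely, while the poof-to-MSTR joints have tight angular limits.
# Both sides share identical shapes and limits.
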PhysicsAsset /BRCosmetics/Characters/Player/Male/Medium/Bodies/M_MED_Cute_Duo/Meshes/Parts/M_MED_Cute_Duo_FaceAcc_Physics
bodies {
 dyn_poof_01_l [{shapeType=Capsule, length=1.048, radius=4.1989, center=<2.0641 -0.0001 0>, rotation=<rot -90 -3.7927 3.7927>}]
 dyn_poof_01_l_MSTR [{shapeType=Capsule, length=0.01, radius=0.01}]
 dyn_poof_01_r [{shapeType=Capsule, length=1.048, radius=4.1989, center=<2.0641 -0.0001 0>, rotation=<rot -90 -3.7927 3.7927>}]
 dyn_poof_01_r_MSTR [{shapeType=Capsule, length=0.01, radius=0.01}]
 head [{shapeType=Capsule, length=5, radius=5}]
}
constraints [
 {
  bone1Name=dyn_poof_01_l_MSTR
  bone2Name=head
  swing1LimitDegrees=45
  swing1Motion=Free
  swing2LimitDegrees=45
  swing2Motion=Free
  twistLimitDegrees=45
  twistMotion=Free
 }
 {bone1Name=dyn_poof_01_l, bone2Name=dyn_poof_01_l_MSTR, swing1LimitDegrees=10, swing2LimitDegrees=10, twistLimitDegrees=8}
 {
  bone1Name=dyn_poof_01_r_MSTR
  bone2Name=head
  swing1LimitDegrees=45
  swing1Motion=Free
  swing2LimitDegrees=45
  swing2Motion=Free
  twistLimitDegrees=45
  twistMotion=Free
 }
 {bone1Name=dyn_poof_01_r, bone2Name=dyn_poof_01_r_MSTR, swing1LimitDegrees=10, swing2LimitDegrees=10, twistLimitDegrees=8}
]
